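# Material function for the Elastic player head that recolors skin and adds an optional eye mask and colored, emissive eyes over incoming material attributes.
# Skin color is modulated by the base color's HSV value to keep texture shading.
# An FX texture masks the mask and eye layers, blended as material attributes, and supplies the eye gradient.
# World position offset collapses UV-targeted eye highlight and vertex-color-tagged hidden vertices toward the object position.
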MaterialFunction /Game/Characters/Player/Female/Medium/Bodies/F_MED_Elastic/Materials/MF_ElasticHead
graph {
 properties {description=SkinColorSaturation}
 nodes [
  Constant2Vector() {desc=uv location of highlight, g=0.615, r=0.978}
  FunctionInput() {inputName=Skin, inputType=FunctionInput_Scalar}
  n1 = FunctionInput() {inputType=FunctionInput_MaterialAttributes}
  n2 = GetMaterialAttributes(None=@n1) {attributeGetTypes [#1]}
  n3 = VectorParameter() {group=Elastic, parameterName=SkinColor, defaultValue=<rgba 0.1089 1 0.0894 0>}
  n4 = Desaturation(None=@n3.RGB, Fraction=ScalarParameter() {defaultValue=0.15, group=Elastic, parameterName=SkinColorDesaturation})
  n5 = MaterialFunctionCall(Float3=MaterialFunctionCall!(In=@n2.BaseColor) {materialFunction=/Engine/Functions/Engine_MaterialFunctions01/Chromakeying/RGBtoHSV}.Result) {materialFunction=&BreakOutFloat3Components}
  n6 = ScalarParameter() {defaultValue=1.3, desc="Mults the V strength of the skin diffuse color. \r\nUsed by the tint color.", group=Elastic, parameterName=SkinVMult}
  n7 = ScalarParameter() {defaultValue=0.9, desc="0=No skin texture used. 1=Value of skin texture multed against skinColor", group=Elastic, parameterName=SkinVStength}
  n8 = LinearInterpolate(A=@n4, B=Multiply(A=Multiply(A=@n5.B, B=@n6), B=@n4), Alpha=@n7) {constAlpha=0.3}
  n9 = TextureSampleParameter2D() {
   desc=red is eyes. Green is eye mask
   group=Elastic
   parameterName=FX
   sortPriority=0
   texture=/Game/Characters/Player/Female/Medium/Heads/F_MED_Handler_Head_01/Skins/Elastic/Textures/T_F_MED_Elastic_Head_FX
  }
  n10 = Multiply(
   A=Multiply(A=Constant() {r=1}, B=OneMinus(None=@n9.R))
   B=Multiply(A=@n3.A, B=ScalarParameter() {defaultValue=1, group=Elastic, parameterName=ColorSkinStrength}) {constB=0.5}
  ) {constB=0.5}
  n11 = VectorParameter() {group=Elastic, parameterName=EyeMask MSRE, defaultValue=<rgba 0.1 0.4 0.4 0.1>}
  n12 = TextureSampleParameter2D() {
   desc=red is eyes. Green is eye mask
   group=Elastic
   parameterName=MaskNormal
   samplerType=SAMPLERTYPE_Normal
   sortPriority=0
   texture=/Game/Characters/Player/Female/Medium/Heads/F_MED_Handler_Head_01/Skins/Elastic/Textures/T_F_MED_Elastic_Head_N
  }
  n13 = SetMaterialAttributes!(
   MaterialAttributes=@n1
   BaseColor=VectorParameter() {group=Elastic, parameterName=AccessoryColor, defaultValue=<rgba 0.01 0.01 0.01 1>}.RGB
   Metallic=@n11.R
   Specular=@n11.R
   Roughness=@n11.R
   Normal=@n12.RGB
  ) {attributeSetTypes [
    #1
    #2
    #3
    #4
    #5
   ]}
  n14 = ScalarParameter() {group=Elastic, parameterName=EyeMask}
  n15 = ScalarParameter() {group=Elastic, parameterName=FaceMask}
  n16 = BlendMaterialAttributes(
   A=SetMaterialAttributes!(MaterialAttributes=@n1, BaseColor=LinearInterpolate(A=@n2.BaseColor, B=@n8, Alpha=@n10)) {attributeSetTypes [#1]}
   B=@n13
   Alpha=Multiply(A=Multiply(A=@n9.G, B=@n14), B=OneMinus(None=@n15))
  )
  n17 = VectorParameter() {group=Elastic, parameterName=EyeColor, defaultValue=<rgba 1 0.01 0 0>}
  n18 = Multiply(A=@n17.RGB) {constB=0.66}
  n19 = VectorParameter() {group=Elastic, parameterName=Eye MSRE, defaultValue=<rgba 0 0.1 0.6 1>}
  n20 = Power(Base=@n17.RGB)
  n21 = Add(A=@n19.A, B=Add(A=Multiply(A=@n9.B) {constB=2.5}) {constB=0.75})
  n22 = Multiply(
   A=LinearInterpolate(A=@n17.A, Alpha=ScalarParameter() {group=Elastic, parameterName=EyeColorAlpha1})
   B=OneMinus(None=ScalarParameter() {group=Elastic, parameterName=ElasticVisor})
  )
  n23 = LinearInterpolate(A=Multiply(A=@n9.R, B=@n22), Alpha=ScalarParameter() {group=Elastic, parameterName=OverrideEyes})
  n24 = SetMaterialAttributes!(
   MaterialAttributes=@n1
   BaseColor=Multiply(A=LinearInterpolate(A=@n18, B=@n17.RGB, Alpha=@n9.B) {desc=gradient}) {constB=0.5}
   Metallic=@n19.R
   Specular=@n19.R
   Roughness=@n19.R
   EmissiveColor=Multiply(A=Multiply(A=@n20, B=@n21), B=@n23)
  ) {attributeSetTypes [
    #1
    #2
    #3
    #4
    #6
   ]}
  n25 = Multiply(A=Subtract(A=WorldPosition().XYZ, B=ObjectPositionWS())) {constB=-1}
  n26 = TextureCoordinate()
  n27 = VectorParameter() {desc=uv location of highlight, group=Elastic, parameterName=Elastic_highlightUVCoord, defaultValue=<rgba 0.978 0.615 0 1>}
  n28 = AppendVector(A=@n27.R, B=@n27.R)
  n29 = ScalarParameter() {defaultValue=0.01, group=Elastic, parameterName=ElasticHighlightRadius, sliderMax=0.06}
  n30 = Max(
   A=Multiply(A=@n22, B=SphereMask(A=@n26, B=@n28, Radius=@n29) {attenuationRadius=0.03})
   B=MaterialFunctionCall!(
    Hide Element Toggle=ScalarParameter() {group=HideElements, parameterName=Hide Element 01}
    VertexColorChannel=VertexColor().R
    Vertex Value to hide=Constant() {r=0.1}
   ) {materialFunction=/Game/Packages/Fortress_SharedMaterials/Base_Material_Functions/VertexColorMask}.Result
  )
  FunctionOutput(None=SetMaterialAttributes!(MaterialAttributes=BlendMaterialAttributes(A=@n16, B=@n24, Alpha=@n23), Normal, BaseColor, WorldPositionOffset=Multiply(A=@n25, B=@n30)) {attributeSetTypes [#5, #1, #7]})
  ScalarParameter() {group=Elastic, parameterName=ColorEyes}
  VectorParameter() {group=Elastic, parameterName=EyeMaskColor, defaultValue=<rgba 0.01 0.01 0.01 1>}
 ]
}
