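# Material function that shades the customizable F_MED_Elastic female player body with two-tone colors, stickers and an accessory material.
# A pattern mask, rounded on some quality levels and mobile, blends primary and secondary materials into ClothFuzz.
# Front and back stickers, then a vertex-color-masked accessory, layer on top.
# View-space directional shading adjusts base and subsurface color, while camera-distance blends tune colored fresnel rim and pattern-edge emissive.
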MaterialFunction /Game/Characters/Player/Female/Medium/Bodies/F_MED_Elastic/Materials/MF_Elastic
graph {nodes [
  n1 = NamedRerouteUsage()
  n2 = NamedRerouteUsage()
  n3 = GetMaterialAttributes(None=FunctionInput() {bUsePreviewValueAsDefault=true, inputName=Material, inputType=FunctionInput_MaterialAttributes}) {attributeGetTypes [#1, #2]}
  n4 = VectorParameter() {group=Elastic, parameterName=StickerPosition, defaultValue=<rgba 0.02 -0.21 0 1>}
  n5 = VectorParameter() {group=Elastic, parameterName=StickerScale, defaultValue=<rgba -0.06 -0.06 0 1>}
  n6 = MaterialFunctionCall(UVs=Add(A=TextureCoordinate(), B=ComponentMask(None=@n4.RGB) {g=true, r=true}), Texture Scale=AppendVector(A=@n5.R, B=@n5.R)) {materialFunction=&ScaleUVsByCenter}
  n7 = TextureSampleParameter2D(UVs=@n6.UVs) {
   group=Elastic
   parameterName=Sticker
   samplerSource=SSM_Clamp_WorldGroupSettings
   texture=/Game/Global/Textures/Default/Blanks/T_Blank_White_ZeroAlpha_sRGBColor
  }
  n8 = VectorParameter() {group=Elastic, parameterName=BackStickerPosition, defaultValue=<rgba 0.32 0.4 0 1>}
  n9 = VectorParameter() {group=Elastic, parameterName=BackStickerScale, defaultValue=<rgba 0.07 0.07 0 1>}
  n10 = MaterialFunctionCall(UVs=Add(A=TextureCoordinate(), B=ComponentMask(None=@n8.RGB) {g=true, r=true}), Texture Scale=AppendVector(A=@n9.R, B=@n9.R)) {materialFunction=&ScaleUVsByCenter}
  n11 = TextureSampleParameter2D(UVs=@n10.UVs) {
   group=Elastic
   parameterName=Sticker
   samplerSource=SSM_Clamp_WorldGroupSettings
   texture=&T_Blank_White_ZeroAlpha_sRGBColor
  }
  n12 = Multiply(A=@n11.A, B=Clamp(None=@n10.0-1 mask))
  n13 = VectorParameter() {group=Elastic, parameterName=Sticker MSRE, defaultValue=<rgba 0 0.2 0.2 0.1>}
  n14 = SetMaterialAttributes!(
   MaterialAttributes=@n3.MaterialAttributes
   BaseColor=Multiply(
    A=LinearInterpolate(A=@n7.RGB, B=@n11.RGB, Alpha=@n12)
    B=ScalarParameter() {defaultValue=1, group=Elastic, parameterName=ElasticStickerMult}
   )
   Metallic=@n13.R
   Specular=@n13.R
   Roughness=@n13.R
   Normal=@n3.MaterialAttributes
  ) {attributeSetTypes [
    #3
    #4
    #5
    #6
    #1
   ]}
  n15 = Add(A=Multiply(A=@n7.A, B=Clamp(None=@n6.0-1 mask)), B=@n12)
  n16 = TextureSampleParameter2D() {
   constMipValue=0
   group=Elastic
   parameterName=Pattern
   samplerType=SAMPLERTYPE_Masks
   texture=/Game/Characters/Player/Female/Medium/Bodies/F_MED_Elastic/Textures/T_F_MED_Elastic_Body_FX
  }
  n17 = Round(None=@n16.RGB)
  n18 = ShadingPathSwitch(Default=QualitySwitch(Default=@n16.RGB, Low=@n17, High=@n17, Epic=@n16.RGB), Mobile=QualitySwitch(Default=@n17, Low=@n17, Medium=@n17))
  n19 = ComponentMask(None=@n18) {r=true}
  n20 = VectorParameter() {group=Elastic, parameterName=PrimaryColor, defaultValue=<rgba 0.05 0.05 0.9 1>}
  n21 = Multiply(A=@n20.RGB, B=NamedRerouteUsage()) {constB=0.8}
  n22 = VectorParameter() {group=Elastic, parameterName=SecondaryColor, defaultValue=<rgba 0.9 0.05 0.05 1>}
  n23 = SphereMask(A=Constant3Vector(), B=Subtract(A=@n20.RGB, B=@n22.RGB)) {attenuationRadius=0.1, hardnessPercent=0}
  n24 = VectorParameter() {desc=MSRE, group=Elastic, parameterName=PrimaryMaterial, defaultValue=<rgba 0 0.5 1 0>}
  n25 = TextureSampleParameter2D() {
   group=Elastic
   parameterName=SpecularMasks
   samplerType=SAMPLERTYPE_Masks
   sortPriority=0
   texture=/Game/Characters/Player/Female/Medium/Bodies/F_MED_Elastic/Textures/T_F_MED_Elastic_Body_S
  }
  n26 = TextureSampleParameter2D() {
   group=Elastic
   parameterName=PrimaryNormal
   samplerType=SAMPLERTYPE_Normal
   texture=/Game/Characters/Player/Female/Medium/Bodies/F_MED_Elastic/Textures/T_F_MED_Elastic_Body_N
  }
  n27 = SetMaterialAttributes(
   MaterialAttributes=@n3.MaterialAttributes
   BaseColor=Multiply(A=LinearInterpolate(A=@n20.RGB, B=@n21, Alpha=@n23)) {constB=1.05}
   Metallic=@n24.R
   Specular=Multiply(A=@n24.G, B=@n25.R)
   Roughness=Multiply(A=@n24.B, B=@n25.B)
   Normal=@n26.RGB
  ) {attributeSetTypes [
    #3
    #4
    #5
    #6
    #1
   ]}
  n28 = Multiply(A=@n22.RGB, B=NamedRerouteUsage())
  n29 = VectorParameter() {desc=MSRE, group=Elastic, parameterName=SecondaryMaterial, defaultValue=<rgba 0 0.5 1 0>}
  n30 = TextureSampleParameter2D() {
   group=Elastic
   parameterName=SecondaryNormal
   samplerType=SAMPLERTYPE_Normal
   texture=&T_F_MED_Elastic_Body_N
  }
  n31 = SetMaterialAttributes(
   MaterialAttributes=@n3.MaterialAttributes
   BaseColor=Multiply(A=LinearInterpolate(A=@n22.RGB, B=@n28, Alpha=@n23)) {constB=0.85}
   Metallic=@n29.R
   Specular=Multiply(A=@n29.G, B=@n25.R)
   Roughness=Multiply(A=@n29.B, B=@n25.B)
   Normal=@n30.RGB
  ) {attributeSetTypes [
    #3
    #4
    #5
    #6
    #1
   ]}
  n32 = GetMaterialAttributes(None=BlendMaterialAttributes(A=@n27, B=@n31, Alpha=@n19)) {attributeGetTypes [#3]}
  n33 = MaterialFunctionCall!(
   Alpha=LinearInterpolate(
    A=ScalarParameter() {group=Elastic, parameterName=PrimaryCloth}
    B=ScalarParameter() {group=Elastic, parameterName=SecondaryCloth}
    Alpha=@n19
   )
   Gmap=Constant() {r=1}
   Material=SetMaterialAttributes!(MaterialAttributes=@n32.MaterialAttributes, BaseColor=@n32.MaterialAttributes) {attributeSetTypes [#3]}
  ) {materialFunction=/Game/Weapons/WeaponSkins/MaterialFunction/ClothFuzz}
  n34 = SphereMask(A=NamedRerouteUsage(), B=Constant() {r=1}) {attenuationRadius=0.1}
  n35 = VectorParameter() {group=Elastic, parameterName=AccessoryMaterial, defaultValue=<rgba 0.5 0.3 0.5 0>}
  n36 = SetMaterialAttributes!(
   MaterialAttributes=@n3.MaterialAttributes
   BaseColor=LinearInterpolate(
    A=VectorParameter() {group=Elastic, parameterName=AccessoryColor, defaultValue=<rgba 0.1 0.1 0.1 1>}.RGB
    B=Constant3Vector() {constant=<rgba 0.1 0.1 0.1 0>}
    Alpha=OneMinus(None=@n34)
   )
   Metallic=@n35.R
   Specular=@n35.R
   Roughness=@n35.R
   Normal=@n3.MaterialAttributes
  ) {attributeSetTypes [
    #3
    #4
    #5
    #6
    #1
   ]}
  n37 = ScalarParameter() {defaultValue=1, group=Elastic, parameterName=UseAccessoryMaterial, sortPriority=0}
  n38 = Round(None=SphereMask(A=VertexColor().R, B=ScalarParameter() {defaultValue=0.1, group=Elastic, parameterName=AccessoryVertexColor}) {attenuationRadius=0.025, hardnessPercent=1})
  n39 = GetMaterialAttributes(None=BlendMaterialAttributes(A=BlendMaterialAttributes(A=@n33.Result, B=@n14, Alpha=@n15) {vertexAttributeBlendType=UseA}, B=@n36, Alpha=Multiply(A=@n37, B=@n38)) {vertexAttributeBlendType=UseA}) {attributeGetTypes [#3, #4, #7]}
  n40 = DotProduct(
   A=PixelNormalWS()
   B=Transform(None=VectorParameter() {group=ElasticRim, parameterName=elastic_TopLightDirection, defaultValue=<rgba 0 0.3 -0.1 0>}.RGB) {transformSourceType=TRANSFORMSOURCE_View}
  )
  n41 = Add(A=@n40) {constB=0.5}
  n42 = ScalarParameter() {parameterName=EnableElasticUpdates}
  n43 = Multiply(A=@n42, B=NamedRerouteUsage())
  n44 = LinearInterpolate(Alpha=@n43) {constA=1.4, constB=1.25}
  n45 = Saturate(None=Power(Base=Multiply(A=@n41, B=@n44) {constB=1.4}))
  n46 = Multiply(A=@n45, B=ScalarParameter() {defaultValue=1.2, parameterName=B}) {constB=1.2}
  n47 = ScalarParameter() {group=ElasticRim, parameterName=elastic_ViewSpaceShading}
  n48 = NamedRerouteUsage()
  n49 = LinearInterpolate(A=LinearInterpolate(A=@n39.SubsurfaceColor, B=Multiply(A=@n39.SubsurfaceColor, B=@n46), Alpha=@n47), Alpha=@n48) {constB=0}
  FunctionOutput(None=SetMaterialAttributes!(
    MaterialAttributes=@n39.MaterialAttributes
    BaseColor=NamedRerouteUsage()
    EmissiveColor=Add(A=@n1, B=@n2)
    SubsurfaceColor=LinearInterpolate(A=@n39.SubsurfaceColor, B=@n49, Alpha=@n42) {desc=Mute bright SSS colors with a bit of shading}
   ) {attributeSetTypes [#3, #2, #7]})
  Multiply(A=NamedRerouteUsage(), B=Constant3Vector() {constant=<rgba 1 0 0 0>})
  n50 = VectorParameter() {group=ElasticRim, parameterName=elasticPrimaryColoredRim, defaultValue=<rgba 1 1 1 1>}
  n51 = ScalarParameter() {defaultValue=0.33, group=ElasticRim, parameterName=elasticPrimaryColoredRimMult}
  n52 = VectorParameter() {group=ElasticRim, parameterName=elasticSecondaryColoredRim, defaultValue=<rgba 1 1 1 1>}
  n53 = ScalarParameter() {defaultValue=0.33, group=ElasticRim, parameterName=elasticSecondaryColoredRimMult}
  n54 = Multiply(
   A=LinearInterpolate(A=Multiply(A=@n50.RGB, B=@n51), B=Multiply(A=@n52.RGB, B=@n53), Alpha=@n19)
   B=ScalarParameter() {defaultValue=0.333, group=ElasticRim, parameterName=elasticColoredRimMult}
  )
  n55 = MaterialFunctionCall!(Blend Range=Constant() {desc=spread, r=3000}, Start Offset=Constant() {desc="start (was -1500)", r=-200}) {materialFunction=/Engine/Functions/Engine_MaterialFunctions03/Particles/Distance_Blend}
  n56 = VectorParameter() {group=ElasticRim, parameterName=elastic_RimDirection, defaultValue=<rgba 1 2 2 0>}
  n57 = VectorParameter() {group=ElasticRim, parameterName=elastic_farRimDirection, defaultValue=<rgba 1 2 2 0>}
  n58 = NamedRerouteUsage()
  n59 = DotProduct(A=PixelNormalWS(), B=Transform(None=LinearInterpolate(A=@n56.RGB, B=@n57.RGB, Alpha=@n58)) {transformSourceType=TRANSFORMSOURCE_View})
  n60 = ScalarParameter() {defaultValue=0.45, group=ElasticRim, parameterName=elastic_ViewSpaceRim}
  n61 = NamedRerouteUsage()
  n62 = OneMinus(None=DotProduct(A=Transform(None=@n61), B=CameraVectorWS()))
  n63 = Add(
   A=Max(A=Multiply(A=@n59, B=@n60) {constB=0}) {constB=0}
   B=Multiply(A=@n62, B=ScalarParameter() {defaultValue=0.333, group=ElasticRim, parameterName=elastic_FresnelRimStrength})
  )
  n64 = Power(
   Base=@n63
   Exp=LinearInterpolate(
    A=ScalarParameter() {defaultValue=2.5, group=ElasticRim, parameterName=elastic_closeRimExponent}
    B=ScalarParameter() {defaultValue=1.5, group=ElasticRim, parameterName=elastic_farRimExponent}
    Alpha=@n55.Result
   ) {constA=2.5, desc=rim exponent - reduces with distance}
  ) {constExponent=2.5}
  n65 = Multiply(A=Power(Base=@n63) {constExponent=1.5}) {constB=0.05}
  n66 = ScalarParameter() {defaultValue=0.5, desc=Fresnel Tint Strength, group=ElasticRim, parameterName=colorRim_close, sliderMin=10}
  n67 = ScalarParameter() {defaultValue=2, desc=Fresnel Tint Strength, group=ElasticRim, parameterName=colorRim_far}
  n68 = NamedRerouteDeclaration(None=Multiply(A=Add(A=@n64, B=@n65), B=LinearInterpolate(A=@n66, B=@n67, Alpha=@n55.Result) {constA=1}) {constB=0.3, desc=white color fresnel}) {name=RimLightMask, nodeColor=<rgba 0.7578 1 0 1>}
  n69 = Multiply(A=Add(A=@n62), B=ScalarParameter() {defaultValue=0.5, group=ElasticRim, parameterName=E_PatternEdgeStrength}) {constB=3}
  n70 = Multiply(
   A=NamedRerouteUsage()
   B=MaterialFunctionCall(Blend Range=Constant() {desc=spread, r=7000}, Start Offset=Constant() {desc=start, r=-1000}) {materialFunction=&Distance_Blend}.Result
  )
  NamedRerouteDeclaration(None=Add(A=Multiply(A=@n54, B=@n68), B=Multiply(A=@n69, B=@n70))) {name=Emissive, nodeColor=<rgba 1 0.7734 0 1>}
  NamedRerouteDeclaration(None=@n3.EmissiveColor) {name=EmissiveInput, nodeColor=<rgba 0 0.0625 1 1>}
  NamedRerouteDeclaration(None=@n3.Normal) {name=InNormal, nodeColor=<rgba 0.7812 1 0 1>}
  n71 = Multiply(A=@n40, B=@n40)
  n72 = Max(
   A=@n39.BaseColor
   B=Multiply(A=VectorParameter() {group=Elastic, parameterName=elastic_minimumMetalicColor, defaultValue=<rgba 0.05 0.05 0.05 0>}.RGB, B=@n39.Metallic)
  )
  n73 = Add(A=@n72, B=Multiply(A=@n43, B=Multiply(A=@n71) {constB=0.5})) {desc=Make dark colors lighter}
  n74 = Multiply(A=@n72, B=@n45)
  NamedRerouteDeclaration(None=LinearInterpolate(A=LinearInterpolate(A=@n73, B=@n74, Alpha=@n47) {constAlpha=0}, Alpha=Multiply(A=@n48, B=@n43)) {constB=0, desc=Exclude RimLight mask from base color}) {name=Color, nodeColor=<rgba 0 0.9062 1 1>}
  NamedRerouteDeclaration(None=LinearInterpolate(
    A=ScalarParameter() {defaultValue=1, group=ElasticRim, parameterName=elastic_closeColorSeparationMult}
    B=ScalarParameter() {defaultValue=1.3, group=ElasticRim, parameterName=elastic_farColorSeparationMult}
    Alpha=@n55.Result
   )) {name=distanceColorSeparationMult, nodeColor=<rgba 0 1 0.0625 1>}
  NamedRerouteDeclaration(None=LinearInterpolate(
    A=ScalarParameter() {defaultValue=1, group=ElasticRim, parameterName=elastic_closeSecondaryColorSeparationMult}
    B=ScalarParameter() {defaultValue=0.2, group=ElasticRim, parameterName=elastic_farColorSecondarySeparationMult}
    Alpha=@n55.Result
   )) {name=distanceColorSecondarySeparationMult, nodeColor=<rgba 0 1 0.0625 1>}
  NamedRerouteDeclaration(None=@n38) {name=BeltMask, nodeColor=<rgba 0.2187 1 0 1>}
  n75 = NamedRerouteDeclaration(None=ComponentMask(None=@n18) {b=true}) {name=Pattern_B, nodeColor=<rgba 1 0.5156 0 1>}
  n76 = NamedRerouteUsage()
  NamedRerouteDeclaration(None=Saturate(None=Subtract(A=@n75, B=@n76))) {name=Pattern_B_noBelt, nodeColor=<rgba 1 0.2812 0 1>}
  NamedRerouteDeclaration(None=@n55.Result) {name=DistanceBlendMask, nodeColor=<rgba 0 0.2117 1 1>}
  NamedRerouteUsage()
  Reroute(None=FunctionInput() {bUsePreviewValueAsDefault=true, inputName=AO, inputType=FunctionInput_Scalar})
  ScalarParameter() {defaultValue=1, parameterName=Alpha}
  SetMaterialAttributes!(MaterialAttributes=@n33.Result, SubsurfaceColor) {attributeSetTypes [#7]}
 ]}
