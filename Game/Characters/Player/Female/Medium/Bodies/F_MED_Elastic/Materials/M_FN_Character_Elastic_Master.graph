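# Master character material for the Elastic medium female player body, set up for skeletal meshes with clothing, morph targets and mesh deformers.
# It chains material functions: the shared character PreFX base feeds MF_Elastic, whose result enters the PostFX base alongside direct PreFX connections.
# It uses material attributes, with the shading model chosen by expression among Default Lit, Subsurface Profile and Cloth.
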Material /Game/Characters/Player/Female/Medium/Bodies/F_MED_Elastic/Materials/M_FN_Character_Elastic_Master
graph {
 properties {
  bUseMaterialAttributes=true
  bUsedWithClothing=true
  bUsedWithMeshDeformer=true
  bUsedWithMorphTargets=true
  bUsedWithSkeletalMesh=true
  pixelDepthOffsetMode=PDOM_Legacy
  shadingModel=MSM_FromMaterialExpression
  usedShadingModels=Default Lit | Subsurface Profile | Cloth
 }
 nodes [
  n1 = MaterialFunctionCall() {materialFunction=&BaseFN_CharacterFunction_PreFX}
  MaterialFunctionCall(
   AO=@n1.AO
   Distance Blend=@n1.AO
   Fresnel=@n1.AO
   Gmap=@n1.AO
   Material=MaterialFunctionCall!(AO=@n1.AO, Material=@n1.AO) {materialFunction=&/MF_Elastic}.Result
  ) {materialFunction=&BaseFN_CharacterFunction_PostFX}
 ]
}
