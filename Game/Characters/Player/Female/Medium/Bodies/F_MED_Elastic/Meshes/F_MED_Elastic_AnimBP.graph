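# Post-process-style corrective AnimBlueprint for the F_MED_Elastic female player body, preserving thigh volume on a pose copied from the attached parent mesh.
# Each thigh_volume bone takes half of its thigh's rotation through a constraint, then receives an additive bone-space offset.
# Each update reads thigh yaw relative to the pelvis as a delta from ref pose, which fades that offset in as yaw goes negative.
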AnimBlueprint /Game/Characters/Player/Female/Medium/Bodies/F_MED_Elastic/Meshes/F_MED_Elastic_AnimBP
target_skeleton=&/F_MED_Elastic_Skeleton
graphs [
 AnimGraph {nodes [
   n1 = "Constraint - thigh_volume_l (R)"(ComponentPose=Local To Component(LocalPose=Copy Pose From Mesh() {bUseAttachedParent=true}), Alpha=0.5, ConstraintWeights_0=1) {
    alphaBoolBlend {blendOption=Linear}
    boneToModify {boneName=thigh_volume_l}
    constraintSetup [{transformType=Rotation, targetBone {boneName=thigh_l}}]
    constraintWeights [1]
   }
   n2 = "Constraint - thigh_volume_r (R)"(ComponentPose=@n1, Alpha=0.5, ConstraintWeights_0=1) {
    alphaBoolBlend {blendOption=Linear}
    boneToModify {boneName=thigh_volume_r}
    constraintSetup [{transformType=Rotation, targetBone {boneName=thigh_r}}]
    constraintWeights [1]
   }
   n3 = "Transform (Modify) Bone - Bone: thigh_volume_l"(ComponentPose=@n2, Alpha=Get thigh_rot_l(), Translation="-3,-5,-1", Rotation="0,0,0", Scale="1,1,1") {
    rotationMode=BMM_Additive
    rotationSpace=BCS_BoneSpace
    scaleMode=BMM_Additive
    scaleSpace=BCS_BoneSpace
    translationMode=BMM_Additive
    translationSpace=BCS_BoneSpace
    alphaBoolBlend {blendOption=Linear}
    alphaScaleBiasClamp {bMapRange=true, inRange {max=0, min=-100}, outRange {max=0, min=1}}
    boneToModify {boneName=thigh_volume_l}
    translation=<-3 -5 -1>
   }
   n4 = "Transform (Modify) Bone - Bone: thigh_volume_r"(ComponentPose=@n3, Alpha=Get thigh_rot_r(), Translation="3,5,1", Rotation="0,0,0", Scale="1,1,1") {
    translationMode=BMM_Additive
    translationSpace=BCS_BoneSpace
    alphaBoolBlend {blendOption=Linear}
    alphaScaleBiasClamp {bMapRange=true, inRange {max=0, min=-100}, outRange {max=0, min=1}}
    boneToModify {boneName=thigh_volume_r}
    translation=<3 5 1>
   }
   Output Pose(Result=Component To Local(ComponentPose=@n4)) {layerGroup=None}
  ]}
 EventGraph {nodes [
   Event BlueprintUpdateAnimation()
   > n1 = Sequence!(execute) -> (then_0, then_1, then_2:Exec, then_3:Exec, then_4:Exec)
   n2 = GetDeltaTransformFromRefPose(self=GetOwnerSkeletalMeshComponent(), BoneName=thigh_r, BaseName=pelvis)
   Set thigh_rot_r(execute=@n1.then_4, thigh_rot_r=BreakRotator(InRot=BreakTransform(InTransform=@n2).Rotation).Yaw)
   n3 = GetDeltaTransformFromRefPose(self=GetOwnerSkeletalMeshComponent(), BoneName=thigh_l, BaseName=pelvis)
   Set thigh_rot_l(execute=@n1.then_3, thigh_rot_l=BreakRotator(InRot=BreakTransform(InTransform=@n3).Rotation).Yaw)
  ]}
]
variables [
 thigh_rot_l {type=number}
 thigh_rot_r {type=number}
]
